ConversationDatabase /NPCLibrary/NPCs/PolarityWinn/NPCLibrary_Conversation_PolarityWinn_Dialogue
entry_points [{tag=Conversation.Entry.PolarityWinn.Dialogue, destinations [#1]}]
nodes [
 ConversationEntryPointNode_0 {
  id=#1
  class=ConversationEntryPointNode
  outputs [#2]
  properties {entryTag {tagName=Conversation.Entry.PolarityWinn.Dialogue}}
 }
 Speech_C_0 {
  id=#2
  class=Speech_C
  outputs [#3, #4, #5, #6]
  properties {nodeName="", softTaskIcon=&T_UI_NPC_question, generalConfig {defaultMessage=What do you need?}}
  sub_nodes [ConversationChoiceNode_5 {class=ConversationChoiceNode, properties {defaultChoiceDisplayText=Ask about...}}]
 }
 Speech_C_5 {
  id=#6
  class=Speech_C
  outputs [#7]
  properties {
   nodeName=""
   softTaskIcon=&T_UI_NPC_question
   generalConfig {defaultMessage="You know, being part of a superhero squad is kind of like being in a band."}
  }
  sub_nodes [ConversationChoiceNode_15 {class=ConversationChoiceNode, properties {defaultChoiceDisplayText=The Supernovas}}]
 }
 Speech_C_6 {
  id=#7
  class=Speech_C
  outputs [#8]
  properties {
   nodeName=""
   softTaskIcon=&T_UI_NPC_question
   generalConfig {defaultMessage="Everybody's got their 'thing'. A band made up of all drums or all guitars would get a bit samey, right?"}
  }
 }
 ReturnToStart_C_23 {
  id=#8
  class=ReturnToStart_C
  properties {softTaskIcon=&T_UI_NPC_Back}
  sub_nodes [ConversationChoiceNode_32 {class=ConversationChoiceNode, properties {defaultChoiceDisplayText=BACK}}]
 }
 Speech_C_1 {
  id=#5
  class=Speech_C
  outputs [#9]
  properties {
   nodeName=""
   softTaskIcon=&T_UI_NPC_question
   generalConfig {defaultMessage="I'm not gonna lie... I find it hard to click into my rhythm, now that the Academy's fallen."}
  }
  sub_nodes [ConversationChoiceNode_1 {class=ConversationChoiceNode, properties {defaultChoiceDisplayText=Supernova Academy}}]
 }
 Speech_C_2 {
  id=#9
  class=Speech_C
  outputs [#10]
  properties {
   nodeName=""
   softTaskIcon=&T_UI_NPC_question
   generalConfig {defaultMessage="The Supernova's are in harmony, but it's like we've lost our bassline."}
  }
 }
 Speech_C_4 {
  id=#10
  class=Speech_C
  outputs [#8]
  properties {
   nodeName=Speech
   softTaskIcon=&T_UI_NPC_question
   generalConfig {defaultMessage=I think we can find it again. We have to.}
  }
 }
 Back_C_0 {
  id=#4
  class=Back_C
  sub_nodes [ConversationChoiceNode_7 {class=ConversationChoiceNode, properties {defaultChoiceDisplayText=BACK}}]
 }
 Speech_C_7 {
  id=#3
  class=Speech_C
  outputs [#11]
  properties {
   nodeName=""
   softTaskIcon=&T_UI_NPC_question
   generalConfig {defaultMessage=I've always had a special connection with music.}
  }
  sub_nodes [ConversationChoiceNode_0 {class=ConversationChoiceNode, properties {defaultChoiceDisplayText=Powers}}]
 }
 Speech_C_15 {
  id=#11
  class=Speech_C
  outputs [#12]
  properties {
   nodeName=""
   softTaskIcon=&T_UI_NPC_question
   generalConfig {defaultMessage="Before all this, I was a DJ, musician, the works."}
  }
 }
 Speech_C_14 {
  id=#12
  class=Speech_C
  outputs [#13]
  properties {
   nodeName=""
   softTaskIcon=&T_UI_NPC_question
   generalConfig {defaultMessage="Now, I've got these amazing powers, and I know I gotta do good with them."}
  }
 }
 Speech_C_3 {
  id=#13
  class=Speech_C
  outputs [#8]
  properties {
   nodeName=Speech
   softTaskIcon=&T_UI_NPC_question
   generalConfig {defaultMessage="After  all, why have superpowers and not try to help people??"}
  }
 }
]
